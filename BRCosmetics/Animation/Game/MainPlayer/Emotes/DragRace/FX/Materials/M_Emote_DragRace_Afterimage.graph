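# Translucent, unlit, two-sided Niagara mesh-particle material for the afterimage effect of the DragRace emote.
# A scaled main texture, scrolled by time in Y and offset by a particle parameter in X, combines with gradient-weighted background into one mask.
# That mask drives a three-color emissive ramp tinted by particle color.
# Opacity multiplies it by a particle-controlled gradient falloff, panning strobe and pixel channels from one packed texture, depth fade and particle alpha.
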Material /BRCosmetics/Animation/Game/MainPlayer/Emotes/DragRace/FX/Materials/M_Emote_DragRace_Afterimage
graph {
 properties {
  bAutomaticallySetUsageInEditor=false
  bUsedWithNiagaraMeshParticles=true
  blendMode=BLEND_Translucent
  shadingModel=MSM_Unlit
  twoSided=true
  usedShadingModels=Unlit
 }
 nodes [
  n1 = VectorParameter() {parameterName=Color1, defaultValue=<rgba 3 0.5143 0 1>}
  n2 = VectorParameter() {parameterName=Color2, defaultValue=<rgba 3 2.5895 0 1>}
  n3 = Multiply(
   A=Add(A=TextureCoordinate()) {constB=-0.5}
   B=AppendVector(A=ScalarParameter() {defaultValue=1, parameterName=UV Scale X}, B=ScalarParameter() {defaultValue=6, parameterName=UV Scale Y})
  )
  n4 = DynamicParameter() {paramNames [Mask, Texture Loop, UV Speed X, Param4]}
  n5 = Multiply(A=ScalarParameter() {defaultValue=1, parameterName=UV Speed Y}, B=Time())
  n6 = TextureSampleParameter2D(UVs=Add(A=Add(A=@n3) {constB=0.5}, B=AppendVector(A=@n4.UV Speed X, B=@n5))) {parameterName=Main Textures, texture=/BRCosmetics/Animation/Game/MainPlayer/Emotes/DragRace/FX/Textures/T_Emote_DragRace_Texture1}
  n7 = MaterialFunctionCall!(
   Mask 1=@n6.A
   Mask 2=@n6.A
   Mask 3=@n6.A
   Mask 4=@n6.A
   Blend=DynamicParameter() {paramNames [Mask, Texture Loop, UV Speed X, Param4]}.Texture Loop
  ) {materialFunction=/Game/Effects/Niagara/MasterMaterials/Functions/Utility/MF_Lerp_4_Scalar_Looping}
  n8 = Multiply(
   A=Power(
    Base=MaterialFunctionCall() {materialFunction=&LinearGradient}.UGradient
    Exp=ScalarParameter() {defaultValue=3.25, parameterName=Color1 Gradient Power}
   )
   B=@n7.Result
  )
  n9 = TextureSampleParameter2D(UVs=TextureCoordinate()) {
   parameterName=Texture BG/Strobe/Pixels
   texture=/BRCosmetics/Animation/Game/MainPlayer/Emotes/DragRace/FX/Textures/T_Emote_DragRace_Texture1BG
  }
  n10 = Multiply(A=@n9.R) {constB=0.5}
  n11 = MaterialFunctionCall() {materialFunction=&LinearGradient}
  n12 = Multiply(A=@n11.UGradient, B=@n11.UGradient)
  n13 = Max(A=@n7.Result, B=Max(A=Multiply(A=@n10, B=@n12), B=Constant() {r=0.125}))
  n14 = MaterialFunctionCall!(
   Color 1=AppendVector(A=VectorParameter() {parameterName=Color3, defaultValue=<rgba 0 1 1 1>}.RGB, B=Constant() {r=1})
   Color 2=AppendVector(A=VectorParameter() {parameterName=Color4, defaultValue=<rgba 0 1 0.9459 1>}.RGB, B=Constant() {r=1})
   Color 3=AppendVector(A=LinearInterpolate(A=@n1.RGB, B=@n2.RGB, Alpha=@n8), B=Constant() {r=1})
   middle=Constant() {r=0.1}
   t=@n13
  ) {materialFunction=/Game/Effects/Niagara/MasterMaterials/Functions/Mojo_MF/MF_FX_Lerp3_Color}
  n15 = ScalarParameter() {defaultValue=1.25, parameterName=Emissive Mult}
  n16 = ParticleColor()
  MaterialFunctionCall(Emissive=Multiply(A=Multiply(A=@n14.Result V3, B=@n15), B=@n16.RGB)) {materialFunction=&MF_FX_EmissiveAdjustments}
  n17 = DynamicParameter() {paramNames [Mask, Texture Loop, UV Speed X, Param4]}
  n18 = ScalarParameter() {parameterName=Strobe1 UV Speed X}
  n19 = ScalarParameter() {defaultValue=0.5, parameterName=Strobe1 UV Speed Y}
  n20 = TextureSampleParameter2D(UVs=Panner(Coordinate=TextureCoordinate(), Speed=AppendVector(A=@n18, B=@n19))) {
   parameterName=Texture BG/Strobe/Pixels
   texture=&T_Emote_DragRace_Texture1BG
  }
  n21 = TextureSampleParameter2D(UVs=TextureCoordinate()) {
   parameterName=Texture BG/Strobe/Pixels
   texture=&T_Emote_DragRace_Texture1BG
  }
  n22 = Multiply(A=Multiply(A=@n13, B=Power(Base=@n11.UGradient, Exp=@n17.Mask)), B=Multiply(A=@n20.G, B=@n21.B))
  Saturate(None=Multiply(A=DepthFade(Opacity=@n22) {fadeDistanceDefault=25}, B=@n16.A))
 ]
}
