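# MF_Lerp_4_Color is a library-exposed material function in the Niagara master material utilities that blends four Vector4 colors with one scalar.
# Blend spans 0 to 3, although its description labels the last segment 3-4; clamped offsets drive three pairwise lerps.
# Nested lerps reuse those alphas to select the segment, so Color 2-to-3 follows a squared curve and Color 3-to-4 a cubed one, while 1-to-2 stays linear.
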MaterialFunction /Game/Effects/Niagara/MasterMaterials/Functions/Utility/MF_Lerp_4_Color
graph {
 properties {bExposeToLibrary=true}
 nodes [
  n1 = FunctionInput() {inputName=Color 2, inputType=FunctionInput_Vector4, sortPriority=1, previewValue=<1 0 0 1>}
  n2 = FunctionInput() {
   description="Blends inputs 0-1 (Color 1 to 2) 1-2 (Color 2 to 3) 3-4 (Color 3 to 4) "
   inputName=Blend
   inputType=FunctionInput_Scalar
   sortPriority=4
   previewValue=<3 0 0 0>
  }
  n3 = LinearInterpolate(A=FunctionInput() {inputName=Color 1, inputType=FunctionInput_Vector4}, B=@n1, Alpha=Clamp(None=@n2))
  n4 = FunctionInput() {inputName=Color 3, inputType=FunctionInput_Vector4, sortPriority=2, previewValue=<0 1 0 1>}
  n5 = Clamp(None=Subtract(A=@n2))
  n6 = LinearInterpolate(A=@n1, B=@n4, Alpha=@n5)
  n7 = Clamp(None=Subtract(A=@n2) {constB=2})
  n8 = LinearInterpolate(A=@n4, B=FunctionInput() {inputName=Color 4, inputType=FunctionInput_Vector4, sortPriority=3, previewValue=<0 0 1 1>}, Alpha=@n7)
  FunctionOutput(None=LinearInterpolate(A=LinearInterpolate(A=@n3, B=@n6, Alpha=@n5), B=LinearInterpolate(A=@n6, B=@n8, Alpha=@n7), Alpha=@n7))
 ]
}
