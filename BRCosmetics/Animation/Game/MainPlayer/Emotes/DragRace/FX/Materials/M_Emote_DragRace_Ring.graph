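# Unlit additive material for the ring effect in the DragRace emote, usable on Niagara mesh particles and skeletal meshes.
# Rotated, scaled texture UVs are offset by time on U and a dynamic parameter on V.
# Texture alpha blends between two four-color palettes, both cycled by the dynamic ColorBlend parameter.
# The resulting color is scaled by time of day and near-camera fade before an emissive-adjustment function.
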Material /BRCosmetics/Animation/Game/MainPlayer/Emotes/DragRace/FX/Materials/M_Emote_DragRace_Ring
graph {
 properties {
  bAutomaticallySetUsageInEditor=false
  bUsedWithMeshDeformer=true
  bUsedWithNiagaraMeshParticles=true
  bUsedWithSkeletalMesh=true
  blendMode=BLEND_Additive
  shadingModel=MSM_Unlit
  usedShadingModels=Unlit
 }
 nodes [
  n1 = VectorParameter() {parameterName=Color1_1, defaultValue=<rgba 0.918 0 1 1>}
  n2 = Constant() {r=1}
  n3 = VectorParameter() {parameterName=Color1_2, defaultValue=<rgba 0.918 0 1 1>}
  n4 = VectorParameter() {parameterName=Color1_3, defaultValue=<rgba 0.918 0 1 1>}
  n5 = VectorParameter() {parameterName=Color1_4, defaultValue=<rgba 0.918 0 1 1>}
  n6 = DynamicParameter() {paramNames [ColorBlend, UV Speed, Param3, Param4]}
  n7 = VectorParameter() {parameterName=Color2_1, defaultValue=<rgba 0.918 0 1 1>}
  n8 = Constant() {r=1}
  n9 = VectorParameter() {parameterName=Color2_2, defaultValue=<rgba 0.918 0 1 1>}
  n10 = VectorParameter() {parameterName=Color2_3, defaultValue=<rgba 0.918 0 1 1>}
  n11 = VectorParameter() {parameterName=Color2_4, defaultValue=<rgba 0.918 0 1 1>}
  n12 = Add(A=MaterialFunctionCall(UVs=TextureCoordinate(), "Rotation Angle (0-1)"=ScalarParameter() {parameterName="Rotation Angle (0-1) (S)"}) {materialFunction=&CustomRotator}.Rotated Values) {constB=-0.5}
  n13 = AppendVector(A=ScalarParameter() {defaultValue=1, parameterName=UV Scale X}, B=ScalarParameter() {defaultValue=1, parameterName=UV Scale Y})
  n14 = ScalarParameter() {defaultValue=1, parameterName=UV Speed Y}
  n15 = Time()
  n16 = Add(
   A=Add(A=Multiply(A=@n12, B=@n13)) {constB=0.5}
   B=AppendVector(A=Multiply(A=@n14, B=@n15), B=DynamicParameter() {paramNames [ColorBlend, UV Speed, Param3, Param4]}.UV Speed)
  )
  n17 = TextureSampleParameter2D(UVs=@n16) {parameterName=Main Textures, texture=/BRCosmetics/Animation/Game/MainPlayer/Emotes/DragRace/FX/Textures/T_Emote_DragRace_Texture1}
  n18 = LinearInterpolate(
   A=MaterialFunctionCall!(
    Color 1=AppendVector(A=@n1.RGB, B=@n2)
    Color 2=AppendVector(A=@n3.RGB, B=@n2)
    Color 3=AppendVector(A=@n4.RGB, B=@n2)
    Color 4=AppendVector(A=@n5.RGB, B=@n2)
    Blend=@n6.ColorBlend
   ) {materialFunction=/Game/Effects/Niagara/MasterMaterials/Functions/Utility/MF_Lerp_4_Color}.Result
   B=MaterialFunctionCall(
    Color 1=AppendVector(A=@n7.RGB, B=@n8)
    Color 2=AppendVector(A=@n9.RGB, B=@n8)
    Color 3=AppendVector(A=@n10.RGB, B=@n8)
    Color 4=AppendVector(A=@n11.RGB, B=@n8)
    Blend=@n6.ColorBlend
   ) {materialFunction=&MF_Lerp_4_Color}.Result
   Alpha=MaterialFunctionCall!(
    Mask 1=@n17.A
    Mask 2=@n17.A
    Mask 3=@n17.A
    Mask 4=@n17.A
    Blend=DynamicParameter() {paramNames [ColorBlend, UV Speed, Param3, Param4]}.ColorBlend
   ) {materialFunction=/Game/Effects/Niagara/MasterMaterials/Functions/Utility/MF_Lerp_4_Scalar_Looping}.Result
  )
  n19 = ScalarParameter() {defaultValue=1, parameterName=Emissive Morning/Evening Mult}
  n20 = MaterialFunctionCall(
   Value_Morning=@n19
   Value_Day=ScalarParameter() {defaultValue=1, parameterName=Emissive Day Mult}
   Value_Evening=@n19
   Value_Night=ScalarParameter() {defaultValue=1, parameterName=Emissive Night Mult}
  ) {materialFunction=&Scalar_Time_of_Day}
  n21 = MaterialFunctionCall(Distance=ScalarParameter() {parameterName=Camera Distance}) {materialFunction=&Near_Camera_Fade}
  n22 = ScalarParameter() {defaultValue=0.4, parameterName=Camera Remap Target Low}
  n23 = ScalarParameter() {defaultValue=1, parameterName=Camera Remap Target High}
  MaterialFunctionCall(Emissive=Multiply(
    A=Multiply(A=@n18, B=@n20.Result)
    B=MaterialFunctionCall(Input=@n21.Result, Target Low=@n22, Target High=@n23) {materialFunction=&RemapValueRange}.Result
   )) {materialFunction=&MF_FX_EmissiveAdjustments}
  MaterialFunctionCall(
   Input=MaterialFunctionCall(Distance=ScalarParameter() {parameterName=Opacity Camera Distance}) {materialFunction=&Near_Camera_Fade}.Result
   Target Low=ScalarParameter() {defaultValue=0.4, parameterName=Opacity Camera Remap Target Low}
   Target High=ScalarParameter() {defaultValue=1, parameterName=Opacity Camera Remap Target High}
  ) {materialFunction=&RemapValueRange}
  ScalarParameter() {defaultValue=1, parameterName=UV Speed X_DragRace}
  ScalarParameter() {parameterName=ColorBlend_DragRace}
  ScalarParameter() {parameterName=ColorBlend_DragRace}
  ShadowReplace(Default=Constant(), Shadow=Constant() {r=1})
  TextureSample(UVs=@n16) {texture=/BRCosmetics/Accessories/FORT_Backpacks/Backpack_M_MED_AlmondSplash/Textures/almondsplash_gradient}
 ]
}
