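# Niagara effect for the DragRace emote, combining swirling colored sprite dots with two gadget lights over a 7.5-second cycle.
# Dots spawn around a sphere, pulled into a vortex while loop-age curves choreograph spawn rate and drag.
# Their colors randomize between two gradients that shift over the loop.
# Two persistent light particles follow the gadget's FX_Light1/FX_Light2 sockets, cycling color, with brightness zeroed in the front end.
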NiagaraSystem /BRCosmetics/Animation/Game/MainPlayer/Emotes/DragRace/FX/NS_Emote_DragRace_Dots
dependencies {
 usedDataInterfaces [
  &NiagaraDataInterfaceCurve
  &NiagaraDataInterfaceSpriteRendererInfo
  &NiagaraDataInterfaceColorCurve
  &NiagaraDataInterfaceSkeletalMesh
 ]
 usedDynamicInputs []
 usedModules [&SystemState]
}
emitters {
 Dots {
  emitterUpdateScript [
   &EmitterState(Life Cycle Mode=::NewEnumerator1, Loop Duration=7.5, UseLoopDelay=false)
   &SpawnRate(
    SpawnRate=&FloatFromCurve(
     FloatCurve=DataInterface() {propertyValues {Curve {keys [
         {interpMode=RCIM_Cubic, leaveTangentWeight=0.5833, tangentMode=RCTM_User, time=0, value=0.4}
         {
          arriveTangent=0.5274
          arriveTangentWeight=0.872
          interpMode=RCIM_Cubic
          leaveTangent=0.5274
          leaveTangentWeight=0.2491
          tangentMode=RCTM_User
          time=1.8
          value=0.65
         }
         {
          arriveTangent=0.4482
          arriveTangentWeight=0.3432
          interpMode=RCIM_Cubic
          leaveTangent=0.4482
          leaveTangentWeight=0.858
          tangentMode=RCTM_User
          time=2.3
          value=0.8
         }
         {
          arriveTangent=0.0102
          arriveTangentWeight=0.4295
          interpMode=RCIM_Cubic
          leaveTangent=0.0102
          leaveTangentWeight=0.2577
          tangentMode=RCTM_User
          time=3.5
          value=1
         }
         {
          arriveTangent=-0.3165
          arriveTangentWeight=0.4664
          interpMode=RCIM_Cubic
          leaveTangent=-0.3165
          leaveTangentWeight=0.7774
          tangentMode=RCTM_User
          time=4.3
          value=0.6
         }
         {
          arriveTangent=-0.0456
          arriveTangentWeight=0.4906
          interpMode=RCIM_Cubic
          leaveTangent=-0.0456
          leaveTangentWeight=0.772
          tangentMode=RCTM_User
          time=5.5
          value=0.5
         }
         {arriveTangentWeight=0.6557, interpMode=RCIM_Cubic, tangentMode=RCTM_User, time=7.5, value=0.4}
        ]}}}
     CurveIndex=$Emitter.LoopedAge
     Scale Curve=36
    )
    Use Spawn Probability=false
   )
  ]
  particleSpawnScript [
   &InitializeParticle(
    Color=<rgba 1.5 1.5 1.5 1>
    Lifetime Max=1.2
    Lifetime Min=0.8
    Lifetime Mode=::NewEnumerator1
    Sprite Rotation Mode=::NewEnumerator2
    Sprite Size Mode=::NewEnumerator1
    Uniform Sprite Size Max=5
    Uniform Sprite Size Min=3
    UsePositionOffset=false
    Write Material Random=true
   )
   &ShapeLocation(
    MaskSpawnGroup=false
    Offset=<0 0 95>
    Offset Mode=::NewEnumerator0
    Override Seed=false
    Sphere Radius=95
    Sphere Surface Distribution=1
    UseOwnerScale=true
   )
   &AddVelocity(
    Module.Use Falloff Distance=false
    Override Seed=false
    Velocity Mode=::NewEnumerator1
    Velocity Origin=$Particles.ShapeLocation.ShapePosition
    Velocity Speed=30
   ) {enabled=false}
   &SubUVAnimation(
    "Sprite Renderer "=DataInterface() {propertyValues {SpriteRenderer=&self:Dots_0.NiagaraSpriteRendererProperties_0}}
    SubUV Animation Mode:ENiagara_SubUVLookupModeV2=::NewEnumerator2
    UseEndFrame=false
    UseStartFrame=false
   )
  ]
  particleUpdateScript [
   &ParticleState()
   &SpriteRotationRate(Rotation Rate=&RandomRangeFloat(Minimum=3, Maximum=15))
   &VortexForce(
    Origin Pull Amount=750
    UseOriginOffset=false
    Vortex Axis=<0 0 -1>
    Vortex Force Amount=&RandomRangeFloat(Minimum=750, Maximum=1000)
   )
   &Drag(
    Drag=&FloatFromCurve(
     FloatCurve=DataInterface() {propertyValues {Curve {keys [
         {interpMode=RCIM_Cubic, leaveTangentWeight=0.6, tangentMode=RCTM_User, time=0, value=1}
         {
          arriveTangent=-0.5274
          arriveTangentWeight=0.6783
          interpMode=RCIM_Cubic
          leaveTangent=-0.5274
          leaveTangentWeight=0.1884
          tangentMode=RCTM_User
          time=1.8
          value=0.75
         }
         {
          arriveTangent=-0.4482
          arriveTangentWeight=0.1826
          interpMode=RCIM_Cubic
          leaveTangent=-0.4482
          leaveTangentWeight=0.4383
          tangentMode=RCTM_User
          time=2.3
          value=0.6
         }
         {
          arriveTangent=-0.0102
          arriveTangentWeight=0.4
          interpMode=RCIM_Cubic
          leaveTangent=-0.0102
          leaveTangentWeight=0.2667
          tangentMode=RCTM_User
          time=3.6
          value=0.4
         }
         {
          arriveTangent=0.3165
          arriveTangentWeight=0.2797
          interpMode=RCIM_Cubic
          leaveTangent=0.3165
          leaveTangentWeight=0.4196
          tangentMode=RCTM_User
          time=4.3
          value=0.8
         }
         {
          arriveTangent=0.0456
          arriveTangentWeight=0.4004
          interpMode=RCIM_Cubic
          leaveTangent=0.0456
          leaveTangentWeight=0.6674
          tangentMode=RCTM_User
          time=5.5
          value=1
         }
         {arriveTangentWeight=0.6667, interpMode=RCIM_Cubic, tangentMode=RCTM_User, time=7.5, value=1}
        ]}}}
     CurveIndex=$Emitter.LoopedAge
     Scale Curve=&RandomRangeFloat(Minimum=1.5, Maximum=2.2)
    )
    Use Linear Drag=true
    Use Rotational Drag=true
   )
   &InheritSourceMovement(
    Applied Position Scale=&VectorFromFloat(Value=&FloatFromCurve(FloatCurve=DataInterface() {propertyValues {Curve {keys [<at 0.7 1>,<at 1 0>]}}}))
    ApplyForce=false
    ApplyPosition=true
    ApplyPositionFalloff=false
   )
   &SolveForcesAndVelocity(Clamp Velocity=false, Limit Acceleration=false)
   &ScaleSpriteSizeBySpeed(Max Scale Factor=<0.2 4>, Velocity Threshold=120) {enabled=false}
   &ScaleColor(
    Scale Alpha=&FloatFromCurve(
     FloatCurve=DataInterface() {propertyValues {Curve {
        defaultValue=0
        keys [
         {interpMode=RCIM_Cubic, tangentMode=RCTM_User, time=0, value=0}
         {interpMode=RCIM_Cubic, tangentMode=RCTM_User, time=0.5, value=1}
         {tangentMode=RCTM_User, time=1, value=0}
        ]
       }}}
     Scale Curve=0.7
    )
    Scale RGB=&RandomRangeVector(
     Minimum=&MakeVectorFromLinearColorRGB(Color=&ColorFromCurve(
       ColorCurve=DataInterface() {propertyValues {
         AlphaCurve {keys [<at 0 1 RCIM_Cubic>]}
         BlueCurve {keys [<at 0 0 RCIM_Cubic>,<at 0.33 1>,<at 0.66 0.128>,<at 1 0 RCIM_Cubic>]}
         GreenCurve {keys [<at 0 0.1714 RCIM_Cubic>,<at 0.33 0.2567>,<at 0.66 0.904>,<at 1 0.1714 RCIM_Cubic>]}
         RedCurve {keys [<at 0 1 RCIM_Cubic>,<at 0.33 0.6953>,<at 0.66 1>,<at 1 1 RCIM_Cubic>]}
        }}
       CurveIndex=$Emitter.NormalizedLoopAge
      ))
     Maximum=&MakeVectorFromLinearColorRGB(Color=&ColorFromCurve(
       ColorCurve=DataInterface() {propertyValues {
         AlphaCurve {keys [<at 0 1 RCIM_Cubic>]}
         BlueCurve {keys [<at 0 1 RCIM_Cubic>,<at 0.33 0.3813>,<at 0.66 0.7605>,<at 1 1 RCIM_Cubic>]}
         GreenCurve {keys [<at 0 0.2582 RCIM_Cubic>,<at 0.33 0.2086>,<at 0.66 0.0409>,<at 1 0.2582 RCIM_Cubic>]}
         RedCurve {keys [<at 0 0 RCIM_Cubic>,<at 0.33 1>,<at 0.66 0.1413>,<at 1 0 RCIM_Cubic>]}
        }}
       CurveIndex=$Emitter.NormalizedLoopAge
      ))
    )
    ScaleA=true
    ScaleRGB=true
   )
  ]
  rendererClasses [&NiagaraSpriteRendererProperties]
  renderers [{rendererClass=&NiagaraSpriteRendererProperties}]
 }
 Light {
  emitterUpdateScript [
   &EmitterState(Life Cycle Mode=::NewEnumerator1, Loop Behavior=::NewEnumerator1, Loop Duration=7.5, UseLoopDelay=false)
   &SpawnBurst_Instantaneous(Use Spawn Probability=false, UseLoopCountLimit=false)
   &SpawnBurst_Instantaneous(Spawn Group=1, Use Spawn Probability=false, UseLoopCountLimit=false) {module=SpawnBurst_Instantaneous001}
  ]
  particleSpawnScript [&InitializeParticle(Lifetime=7.5, Mesh Scale Mode=::NewEnumerator3, UsePositionOffset=false, Write Material Random=true)]
  particleUpdateScript [
   &ParticleState(Kill Particles When Lifetime Has Elapsed=false, Loop Particles Lifetime=true)
   &SkeletalMeshLocation(
    Bone / Socket Index=&SpawnGroup()
    MaskSpawnGroup=false
    Mesh Sampling Type=::NewEnumerator5
    Skeletal Mesh=DataInterface() {propertyValues {
      DefaultMesh=null
      ExcludeBoneName=root
      OptionalComponentClass=null
      PreviewMesh=/BRCosmetics/Animation/Props/Gadgets_DragRace/Meshes/DragRace_Gadget
      SkinningMode=SkinOnTheFly
      SoftSourceActor=null
      SourceMode=Default
      UvSetIndex=0
      WholeMeshLOD=-1
      bExcludeBone=false
      bReadDeformedGeometry=true
      bRequireCurrentFrameData=true
      ComponentTags []
      FilteredBones [JNT_Hoop_01]
      FilteredSockets [FX_Light1,FX_Light2]
      MeshUserParameter {parameter {name=None, typeDefHandle {registeredTypeIndex=1}, varData []}}
      SamplingRegions []
     }}
    Socket Sampling Mode=::NewEnumerator5
   )
   &FN_Light_Attributes(
    Color=&ColorFromCurve(ColorCurve=DataInterface() {propertyValues {
       AlphaCurve {keys [<at 0 1 RCIM_Cubic>,<at 1 1 RCIM_Cubic>]}
       BlueCurve {keys [<at 0 1 RCIM_Cubic>,<at 0.33 0.3813>,<at 0.66 1>,<at 1 1 RCIM_Cubic>]}
       GreenCurve {keys [<at 0 0.3419 RCIM_Cubic>,<at 0.33 0.2086>,<at 0.66 0.2083>,<at 1 0.3419 RCIM_Cubic>]}
       RedCurve {keys [<at 0 0.1812 RCIM_Cubic>,<at 0.33 1>,<at 0.66 0.3187>,<at 1 0.1812 RCIM_Cubic>]}
      }})
    Create Color Attribute=true
    Create Exponent Attribute=true
    Create Position Attribute=true
    Create Radius Attribute=true
    Exponent=2
    Position:NiagaraPosition=$Particles.SkeletalMeshLocation.SampledPosition
    Radius=180
    Scale Color=&FN_FrontEndFloatModifier(Gameplay=4, Front End=0)
   )
  ]
  rendererClasses [&NiagaraLightRendererProperties]
  renderers [{rendererClass=&NiagaraLightRendererProperties, rendererIndex=0}]
 }
}
summary {
 systemName=NS_Emote_DragRace_Dots
 emitters [
  {emitterName=Dots, rendererClasses [&NiagaraSpriteRendererProperties]}
  {emitterName=Light, rendererClasses [&NiagaraLightRendererProperties]}
 ]
 userVariables [
  User.bIsFrontEnd {type=NiagaraBool}
  User.bIsFrontEndPreview {type=NiagaraBool}
  User.ColorBlend {type=NiagaraFloat, defaultValue=0, description=""}
 ]
}
