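# Niagara system for the DragRace emote's ring effect: looping mesh particles raised above the origin on a 7.5-second cycle.
# Ring burst-spawns into two mesh renderers, animating three color cycles, dynamic material parameters and a tilt from loop-age curves.
# Its alpha combines waveform flicker with an owner-position camera fade.
# Ring001 is a simpler companion mesh emitter with its own material-parameter curves, without color or orientation animation.
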NiagaraSystem /BRCosmetics/Animation/Game/MainPlayer/Emotes/DragRace/FX/NS_Emote_DragRace_Ring
dependencies {
 usedDataInterfaces [&NiagaraDataInterfaceColorCurve, &NiagaraDataInterfaceCurve, &NiagaraDataInterfaceCamera]
 usedDynamicInputs []
 usedModules [&SystemState]
}
emitters {
 Ring {
  emitterUpdateScript [
   &EmitterState(Life Cycle Mode=::NewEnumerator1, Loop Behavior=::NewEnumerator1, Loop Duration=7.5, UseLoopDelay=false)
   &SpawnBurst_Instantaneous(Use Spawn Probability=false, UseLoopCountLimit=false)
   &SpawnBurst_Instantaneous(Spawn Group=1, Use Spawn Probability=false, UseLoopCountLimit=false) {enabled=false, module=SpawnBurst_Instantaneous001}
   /SetVariables(
    Emitter.Color1:LinearColor=&MultiplyLinearColor(
     Linear Color=&ColorFromCurve(
      ColorCurve=DataInterface() {propertyValues {
        AlphaCurve {keys [<at 0 1 RCIM_Cubic>,<at 1 0 RCIM_Cubic>]}
        BlueCurve {keys [<at 0 0 RCIM_Cubic>,<at 0.33 1>,<at 0.66 0>,<at 1 0 RCIM_Cubic>]}
        GreenCurve {keys [<at 0 0.1714 RCIM_Cubic>,<at 0.33 0>,<at 0.66 0.8899>,<at 1 0.1714 RCIM_Cubic>]}
        RedCurve {keys [<at 0 1 RCIM_Cubic>,<at 0.33 0.5901>,<at 0.66 1>,<at 1 1 RCIM_Cubic>]}
       }}
      CurveIndex=$Emitter.NormalizedLoopAge
     )
     Scale Factor=&MakeVector4_FromFloat(Float=5)
    )
    Emitter.Color2:LinearColor=&MultiplyLinearColor(
     Linear Color=&ColorFromCurve(
      ColorCurve=DataInterface() {propertyValues {
        AlphaCurve {keys [<at 0 1 RCIM_Cubic>,<at 1 0 RCIM_Cubic>]}
        BlueCurve {keys [<at 0 0 RCIM_Cubic>,<at 0.33 0>,<at 0.66 0.0205>,<at 1 0 RCIM_Cubic>]}
        GreenCurve {keys [<at 0 0.8632 RCIM_Cubic>,<at 0.33 0.2298>,<at 0.66 0>,<at 1 0.8632 RCIM_Cubic>]}
        RedCurve {keys [<at 0 1 RCIM_Cubic>,<at 0.33 1>,<at 0.66 1>,<at 1 1 RCIM_Cubic>]}
       }}
      CurveIndex=$Emitter.NormalizedLoopAge
     )
     Scale Factor=&MakeVector4_FromFloat(Float=2)
    )
    Emitter.Color3_4:LinearColor=&MultiplyLinearColor(
     Linear Color=&ColorFromCurve(
      ColorCurve=DataInterface() {propertyValues {
        AlphaCurve {keys [<at 0 1 RCIM_Cubic>,<at 1 0 RCIM_Cubic>]}
        BlueCurve {keys [<at 0 1 RCIM_Cubic>,<at 0.33 0.2182>,<at 0.66 1>,<at 1 1 RCIM_Cubic>]}
        GreenCurve {keys [<at 0 0.2582 RCIM_Cubic>,<at 0.33 0>,<at 0.66 0>,<at 1 0.2582 RCIM_Cubic>]}
        RedCurve {keys [<at 0 0 RCIM_Cubic>,<at 0.33 1>,<at 0.66 0.1394>,<at 1 0 RCIM_Cubic>]}
       }}
      CurveIndex=$Emitter.NormalizedLoopAge
     )
     Scale Factor=&MakeVector4_FromFloat(Float=2)
    )
   ) {isSetParameters=true}
  ]
  particleSpawnScript [
   /SetVariables(Particles.VisibilityTag=&SpawnGroup()) {enabled=false, isSetParameters=true}
   &InitializeParticle(Lifetime=7.5, Mesh Scale Mode=::NewEnumerator3, Position Offset=<0 0 105>, UsePositionOffset=true, Write Material Random=true)
  ]
  particleUpdateScript [
   &ParticleState(Kill Particles When Lifetime Has Elapsed=false, Loop Particles Lifetime=true)
   &DynamicMaterialParameters(
    Index 0 Param 1=&FloatFromCurve(
     FloatCurve=DataInterface() {propertyValues {Curve {keys [
         <at 0 9 RCIM_Cubic>
         {arriveTangent=0.0586, interpMode=RCIM_Cubic, leaveTangent=0.0586, tangentMode=RCTM_User, time=0.1, value=8.1165}
         {arriveTangent=-1.875, interpMode=RCIM_Cubic, leaveTangent=-1.875, time=0.3, value=18}
         {arriveTangent=-2.0444, interpMode=RCIM_Cubic, leaveTangent=-2.0444, tangentMode=RCTM_User, time=0.5, value=7.3665}
         {arriveTangent=0.2614, interpMode=RCIM_Cubic, leaveTangent=0.2614, tangentMode=RCTM_User, time=1.7, value=5.711}
         {arriveTangent=96.4452, interpMode=RCIM_Cubic, leaveTangent=96.4452, time=1.8, value=25}
         {arriveTangent=0.0225, interpMode=RCIM_Cubic, leaveTangent=0.0225, tangentMode=RCTM_User, time=1.9, value=25}
         {arriveTangent=-2.3835, interpMode=RCIM_Cubic, leaveTangent=-2.3835, tangentMode=RCTM_User, time=2.05, value=4.711}
         {arriveTangent=-1.9731, interpMode=RCIM_Cubic, leaveTangent=-1.9731, time=2.65, value=4}
         {arriveTangent=0.9259, interpMode=RCIM_Cubic, leaveTangent=0.9259, time=3.5, value=1.85}
         {arriveTangent=2.575, interpMode=RCIM_Cubic, leaveTangent=2.575, time=4, value=5.25}
         {arriveTangent=1.0714, interpMode=RCIM_Cubic, leaveTangent=1.0714, time=5.5, value=7}
         <at 7.5 9 RCIM_Cubic>
        ]}}}
     CurveIndex=$Emitter.LoopedAge
    )
    Index 0 Param 2=&FloatFromCurve(FloatCurve=DataInterface() {propertyValues {Curve {keys [<at 0 0>,<at 1 1>]}}}, CurveIndex=$Emitter.NormalizedLoopAge, Scale Curve=4)
    Index 0 Param 3=&FloatFromCurve(
     FloatCurve=DataInterface() {propertyValues {Curve {keys [
         {arriveTangent=1.3557, interpMode=RCIM_Cubic, leaveTangent=1.3557, tangentMode=RCTM_User, time=0, value=1.25}
         <at 1.4 2 RCIM_Constant>
         <at 1.5 0 RCIM_Constant>
         {interpMode=RCIM_Cubic, tangentMode=RCTM_User, time=3.4, value=0}
         <at 5.2 2 RCIM_Constant>
         <at 5.3 0 RCIM_Constant>
         {interpMode=RCIM_Cubic, tangentMode=RCTM_User, time=7, value=0}
         <at 7.5 1 RCIM_Cubic>
        ]}}}
     CurveIndex=$Emitter.LoopedAge
    )
    Module.Index0Float0Write=true
    Module.Index0Float1Write=true
    Module.Index0Float2Write=true
    Module.Index0Float3Write=true
   )
   &ScaleColor(
    Scale Alpha=&Multiply_Float(
     A=&Waveform("Waveform [1]"=::NewEnumerator2, "[1] Frequency"=3, "[1] Compound Frequency"=4, "[1] Amplitude Min/Max"=<0.25 1>)
     B=&Lerp_Float(
      A=0.05
      B=1
      Alpha=/Niagara/DynamicInputs/Camera/CameraFade(
       Camera Query:NiagaraDataInterfaceCamera=DataInterface()
       Fade Offset:NiagaraFloat=125
       Fade Distance:NiagaraFloat=225
       Use Owner Position:NiagaraBool=true
      )
     )
    )
    Scale RGB=&VectorFromFloat(Value=3)
    ScaleA=true
    ScaleRGB=true
   )
   /SetVariables(Particles.CrossColor:LinearColor=&MakeLinearColorFromVectorAndFloat(
     "Vector (RGB)"=<1 1 1>
     "Float (Alpha)"=&FloatFromCurve(
      FloatCurve=DataInterface() {propertyValues {Curve {keys [
          <at 0 1>
          {arriveTangent=-1, interpMode=RCIM_Cubic, leaveTangent=-1, time=0.75, value=1}
          {arriveTangent=0.0004, interpMode=RCIM_Cubic, leaveTangent=0.0004, tangentMode=RCTM_User, time=1, value=0}
          {arriveTangent=0.0016, interpMode=RCIM_Cubic, leaveTangent=0.0016, tangentMode=RCTM_User, time=6.75, value=0}
          <at 7.5 1 RCIM_Cubic>
         ]}}}
      CurveIndex=$Emitter.LoopedAge
     )
    )) {enabled=false, isSetParameters=true}
   &ScaleColor(ScaleA=true, ScaleRGB=true) {enabled=false, module=ScaleColor001}
   /SetVariables(Particles.MeshOrientation=&MakeQuaternion(
     Quaternion From=0
     Axis=<0 1 0>
     Angle=&FloatFromCurve(
      FloatCurve=DataInterface() {propertyValues {Curve {keys [
          {arriveTangent=0.1676, interpMode=RCIM_Cubic, leaveTangent=0.1676, tangentMode=RCTM_User, time=0, value=-0.015}
          {arriveTangent=0.06, interpMode=RCIM_Cubic, leaveTangent=0.06, time=0.1, value=0.35}
          {arriveTangent=-3.3047, interpMode=RCIM_Cubic, leaveTangent=-3.3047, tangentMode=RCTM_User, time=0.25, value=0}
          {arriveTangent=0.2067, interpMode=RCIM_Cubic, leaveTangent=0.2067, tangentMode=RCTM_User, time=0.3, value=-0.125}
          {interpMode=RCIM_Cubic, leaveTangent=0.0232, time=0.375, value=-0.015}
         ]}}}
      CurveIndex=$Emitter.NormalizedLoopAge
     )
     From=?
     To=?
     Angle Type=::NewEnumerator1
     Normalize Vectors=false
    )) {isSetParameters=true}
  ]
  rendererClasses [&NiagaraMeshRendererProperties]
  renderers [
   {rendererClass=&NiagaraMeshRendererProperties, rendererIndex=0}
   {rendererClass=&NiagaraMeshRendererProperties, rendererIndex=1}
  ]
 }
 Ring001 {
  emitterUpdateScript [
   &EmitterState(Life Cycle Mode=::NewEnumerator1, Loop Behavior=::NewEnumerator1, Loop Duration=7.5, UseLoopDelay=false)
   &SpawnBurst_Instantaneous(Use Spawn Probability=false, UseLoopCountLimit=false)
  ]
  particleSpawnScript [&InitializeParticle(Lifetime=7.5, Mesh Scale Mode=::NewEnumerator3, Position Offset=<0 0 105>, UsePositionOffset=true, Write Material Random=true)]
  particleUpdateScript [
   &ParticleState(Kill Particles When Lifetime Has Elapsed=false, Loop Particles Lifetime=true)
   &DynamicMaterialParameters(
    Index 0 Param 1=&FloatFromCurve(FloatCurve=DataInterface() {propertyValues {Curve {keys [<at 0 0>,<at 1 4>]}}})
    Index 0 Param 2=&FloatFromCurve(FloatCurve=DataInterface() {propertyValues {Curve {
        defaultValue=0
        keys [
         <at 0 1.25 RCIM_Cubic>
         <at 0.188 2 RCIM_Constant>
         <at 0.206 0 RCIM_Constant>
         {arriveTangent=3.5778, interpMode=RCIM_Cubic, leaveTangent=3.5778, time=0.471, value=0}
         <at 0.765 2 RCIM_Constant>
         <at 0.782 0 RCIM_Constant>
         {arriveTangent=6.2893, interpMode=RCIM_Cubic, leaveTangent=6.2893, time=0.882, value=0}
         <at 0.941 1>
        ]
       }}})
    Module.Index0Float0Write=true
    Module.Index0Float1Write=true
    Module.Index0Float2Write=true
    Module.Index0Float3Write=true
   )
  ]
  rendererClasses [&NiagaraMeshRendererProperties]
  renderers [{rendererClass=&NiagaraMeshRendererProperties, rendererIndex=0}]
 }
}
summary {
 systemName=NS_Emote_DragRace_Ring
 emitters [
  {emitterName=Ring, rendererClasses [&NiagaraMeshRendererProperties]}
  {emitterName=Ring001, rendererClasses [&NiagaraMeshRendererProperties]}
 ]
 userVariables [
  User.ColorBlend_DragRace {type=NiagaraFloat, defaultValue=0, description=""}
  User.UV Speed X_DragRace {type=NiagaraFloat, defaultValue=0, description=""}
 ]
}
