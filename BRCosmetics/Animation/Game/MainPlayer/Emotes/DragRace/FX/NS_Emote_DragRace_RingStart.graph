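# Niagara system for the start-ring effect of the DragRace emote, with two burst emitters that each render meshes and lights.
# Ring waits briefly, then animates a material parameter, alpha and slight mesh rotation, plus three fading emitter colors, mostly keyed to normalized loop age.
# Ring001 quickly scales its mesh up from zero and drives a blue light whose color scale ramps in gameplay but is zero in the front end.
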NiagaraSystem /BRCosmetics/Animation/Game/MainPlayer/Emotes/DragRace/FX/NS_Emote_DragRace_RingStart
dependencies {
 usedDataInterfaces [&NiagaraDataInterfaceColorCurve, &NiagaraDataInterfaceSkeletalMesh, &NiagaraDataInterfaceCurve]
 usedDynamicInputs []
 usedModules [&SystemState]
}
emitters {
 Ring {
  emitterUpdateScript [
   &EmitterState(Life Cycle Mode=::NewEnumerator1, Loop Behavior=::NewEnumerator1, Loop Delay=0.43, Loop Duration=0.57, UseLoopDelay=true)
   &SpawnBurst_Instantaneous(Use Spawn Probability=false, UseLoopCountLimit=false)
   /SetVariables(
    Emitter.Color1:LinearColor=&ColorFromCurve(
     ColorCurve=DataInterface() {propertyValues {
       AlphaCurve {keys [<at 0 1 RCIM_Cubic>,<at 1 0 RCIM_Cubic>]}
       BlueCurve {keys [<at 0 0 RCIM_Cubic>]}
       GreenCurve {keys [<at 0 0.1714 RCIM_Cubic>]}
       RedCurve {keys [<at 0 1 RCIM_Cubic>]}
      }}
     CurveIndex=$Emitter.NormalizedLoopAge
    )
    Emitter.Color2:LinearColor=&ColorFromCurve(
     ColorCurve=DataInterface() {propertyValues {
       AlphaCurve {keys [<at 0 1 RCIM_Cubic>,<at 1 0 RCIM_Cubic>]}
       BlueCurve {keys [<at 0 0 RCIM_Cubic>]}
       GreenCurve {keys [<at 0 0.8632 RCIM_Cubic>]}
       RedCurve {keys [<at 0 1 RCIM_Cubic>]}
      }}
     CurveIndex=$Emitter.NormalizedLoopAge
    )
    Emitter.Color3_4:LinearColor=&ColorFromCurve(
     ColorCurve=DataInterface() {propertyValues {
       AlphaCurve {keys [<at 0 1 RCIM_Cubic>,<at 1 0 RCIM_Cubic>]}
       BlueCurve {keys [<at 0 1 RCIM_Cubic>]}
       GreenCurve {keys [<at 0 0.0887 RCIM_Cubic>]}
       RedCurve {keys [<at 0 0 RCIM_Cubic>]}
      }}
     CurveIndex=$Emitter.NormalizedLoopAge
    )
   ) {isSetParameters=true}
  ]
  particleSpawnScript [
   &InitializeParticle(Lifetime=0.57, Mesh Scale Mode=::NewEnumerator3, Position Offset=<0 0 105>, UsePositionOffset=true, Write Material Random=true)
   &InitialMeshOrientation(Mesh Orientation Mode=::NewEnumerator2, Use Rotation Vector=true) {enabled=false}
  ]
  particleUpdateScript [
   &ParticleState()
   &SkeletalMeshLocation(
    Bone Sampling Mode=::NewEnumerator4
    MaskSpawnGroup=false
    Sampled Position Offset=<0 0 96>
    Skeletal Mesh=DataInterface() {propertyValues {
      DefaultMesh=null
      ExcludeBoneName=root
      OptionalComponentClass=null
      PreviewMesh=/BRCosmetics/Animation/Props/Gadgets_DragRace/Meshes/DragRace_Gadget
      SkinningMode=SkinOnTheFly
      SoftSourceActor=null
      SourceMode=Default
      UvSetIndex=0
      WholeMeshLOD=-1
      bExcludeBone=false
      bReadDeformedGeometry=true
      bRequireCurrentFrameData=true
      ComponentTags []
      FilteredBones [JNT_Hoop_01]
      FilteredSockets []
      MeshUserParameter {parameter {name=None, typeDefHandle {registeredTypeIndex=1}, varData []}}
      SamplingRegions []
     }}
   ) {enabled=false}
   &DynamicMaterialParameters(
    Index 0 Param 1=&FloatFromCurve(
     FloatCurve=DataInterface() {propertyValues {Curve {keys [<at 0 25 RCIM_Cubic>,<at 0.8 9 RCIM_Cubic>]}}}
     CurveIndex=$Emitter.NormalizedLoopAge
    )
    Module.Index0Float0Write=true
    Module.Index0Float1Write=true
    Module.Index0Float2Write=true
    Module.Index0Float3Write=true
   )
   &ScaleColor(
    Scale Alpha=&FloatFromCurve(FloatCurve=DataInterface() {propertyValues {Curve {
        defaultValue=0
        keys [
         {interpMode=RCIM_Cubic, tangentMode=RCTM_User, time=0, value=0}
         {interpMode=RCIM_Cubic, tangentMode=RCTM_User, time=0.5, value=1}
         {tangentMode=RCTM_User, time=1, value=1}
        ]
       }}})
    ScaleA=true
    ScaleRGB=true
   )
   /SetVariables(Particles.MeshOrientation=&MakeQuaternion(
     Quaternion From=0
     Axis=<0 1 0>
     Angle=&FloatFromCurve(
      FloatCurve=DataInterface() {propertyValues {Curve {keys [<at 0 0 RCIM_Cubic>,<at 1 -0.015 RCIM_Cubic>]}}}
      CurveIndex=$Emitter.NormalizedLoopAge
     )
     From=?
     To=?
     Angle Type=::NewEnumerator1
     Normalize Vectors=false
    )) {isSetParameters=true}
  ]
  rendererClasses [&NiagaraMeshRendererProperties, &NiagaraLightRendererProperties]
  renderers [
   {rendererClass=&NiagaraMeshRendererProperties, rendererIndex=0}
   {rendererClass=&NiagaraLightRendererProperties, rendererIndex=1}
  ]
 }
 Ring001 {
  emitterUpdateScript [
   &EmitterState(Life Cycle Mode=::NewEnumerator1, Loop Behavior=::NewEnumerator1, Loop Duration=1, UseLoopDelay=false)
   &SpawnBurst_Instantaneous(Use Spawn Probability=false, UseLoopCountLimit=false)
  ]
  particleSpawnScript [&InitializeParticle(Mesh Scale Mode=::NewEnumerator3, UsePositionOffset=true, Write Material Random=true)]
  particleUpdateScript [
   &ParticleState()
   &ScaleMeshSize(Scale Factor=&VectorFromFloat(Value=&FloatFromCurve(FloatCurve=DataInterface() {propertyValues {Curve {defaultValue=1, keys [<at 0 0 RCIM_Cubic>,<at 0.1 1 RCIM_Cubic>]}}})))
   &DynamicMaterialParameters(Module.Index0Float0Write=true, Module.Index0Float1Write=true, Module.Index0Float2Write=true, Module.Index0Float3Write=true)
   &FN_Light_Attributes(
    Color=<rgba 0.1812 0.3419 1 1>
    Create Color Attribute=true
    Create Exponent Attribute=true
    Create Position Attribute=true
    Create Radius Attribute=true
    Exponent=2
    Position=<0 0 15>
    Radius=180
    Scale Color=&FN_FrontEndFloatModifier(
     Gameplay=&FloatFromCurve(
      FloatCurve=DataInterface() {propertyValues {Curve {defaultValue=1, keys [{tangentMode=RCTM_User, time=0, value=0}, {tangentMode=RCTM_User, time=0.5, value=1}]}}}
      Scale Curve=4
     )
     Front End=0
    )
   )
  ]
  rendererClasses [&NiagaraMeshRendererProperties, &NiagaraLightRendererProperties]
  renderers [
   {rendererClass=&NiagaraMeshRendererProperties, rendererIndex=0}
   {rendererClass=&NiagaraLightRendererProperties, rendererIndex=1}
  ]
 }
}
summary {
 systemName=NS_Emote_DragRace_RingStart
 emitters [
  {emitterName=Ring, rendererClasses [&NiagaraMeshRendererProperties, &NiagaraLightRendererProperties]}
  {emitterName=Ring001, rendererClasses [&NiagaraMeshRendererProperties, &NiagaraLightRendererProperties]}
 ]
 userVariables [
  User.bIsFrontEnd {type=NiagaraBool}
  User.bIsFrontEndPreview {type=NiagaraBool}
 ]
}
